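# Translucent, two-sided unlit material for the Storm Umbrella glider's swirl effect, usable on Niagara mesh particles.
# A scrolling UV gradient, offset by vertex color red, forms smoothstepped bands, capped by particle FadeIn and masked by a swirl texture.
# A scrolling normal map dotted with sunlight direction is quantized into steps blending two colors, faking cel-shaded lighting for emissive.
# A time-driven sine along a direction vector supplies wobble.
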Material /Game/Gadgets/Assets/VinderTech_Umbrella/Umbrella_Storm/Materials/M_StormUmbrellaSwirls
graph {
 properties {
  bCastRayTracedShadows=false
  bEnableResponsiveAA=true
  bUsedWithNiagaraMeshParticles=true
  blendMode=BLEND_Translucent
  ditherOpacityMask=true
  pixelDepthOffsetMode=PDOM_Legacy
  shadingModel=MSM_Unlit
  translucencyLightingMode=TLM_SurfacePerPixelLighting
  twoSided=true
  usedShadingModels=Unlit
  translucentMultipleScatteringExtinction=<rgba 0.2296 0.2754 0.599 1>
 }
 nodes [
  Constant()
  n1 = TextureCoordinate()
  n2 = ComponentMask(None=@n1) {r=true}
  n3 = Add(
   A=Multiply(A=ScalarParameter() {defaultValue=1, parameterName=UWeight}, B=OneMinus(None=@n2))
   B=Multiply(A=ComponentMask(None=@n1) {g=true}, B=ScalarParameter() {defaultValue=1, parameterName=VWeight})
  )
  n4 = Multiply(A=Time(), B=ScalarParameter() {defaultValue=1, parameterName=Speed})
  n5 = VertexColor()
  n6 = ScalarParameter() {parameterName=VertexColorRedOffset}
  n7 = Frac(None=Add(A=Add(A=@n3, B=@n4), B=Multiply(A=@n5.R, B=@n6)))
  n8 = ScalarParameter() {defaultValue=0.5, parameterName=DistanceReference}
  n9 = Power(Base=OneMinus(None=Distance(A=@n7, B=@n8)), Exp=ScalarParameter() {defaultValue=1, parameterName=OpacityPow, sortPriority=1})
  MaterialFunctionCall(In=@n9, Contrast=ScalarParameter() {parameterName=Contrast}) {materialFunction=&CheapContrast}
  MaterialFunctionCall() {materialFunction=&LocalPosition}
  n10 = CollectionParameter() {
   collection=&FortniteMaterialParameters
   parameterName=SunlightDirectionalVector
  }
  n11 = ComponentMask(None=@n10) {b=true, g=true, r=true}
  n12 = TextureCoordinate()
  n13 = Time()
  n14 = Multiply(
   A=MaterialFunctionCall(X=@n13, Y=@n13) {materialFunction=&MakeFloat2}.Result
   B=ComponentMask(None=VectorParameter() {parameterName=NormalScrollSpeed}.RGB) {g=true, r=true}
  )
  n15 = VectorParameter() {parameterName=NormalUVScale, defaultValue=<rgba 1 1 0 0>}
  n16 = TextureSampleParameter2D(UVs=Multiply(A=Add(A=@n12, B=@n14), B=ComponentMask(None=@n15.RGB) {g=true, r=true})) {parameterName=NormalMap, samplerType=SAMPLERTYPE_Normal, texture=/Game/Packages/Fortnite_UniqueMaterials_1/Textures/T_WaterFall_Tile_N}
  n17 = Transform(None=@n16.RGB)
  n18 = ScalarParameter() {defaultValue=12, parameterName=Quant}
  n19 = Multiply(A=Saturate(None=DotProduct(A=@n11, B=@n17)), B=@n18)
  n20 = LinearInterpolate(
   A=VectorParameter() {parameterName=Color1, defaultValue=<rgba 0.0762 0.0648 0.2307 1>}.RGB
   B=VectorParameter() {parameterName=Color2, defaultValue=<rgba 0.2159 0.4564 0.6654 1>}.RGB
   Alpha=Divide(A=Round(None=@n19), B=@n18)
  )
  n21 = ScalarParameter() {defaultValue=1, parameterName=DiffuseMultiplier}
  MaterialFunctionCall(Emissive=Multiply(A=Multiply(A=@n20, B=@n21), B=ScalarParameter() {defaultValue=1, parameterName=EmissiveIntensity})) {materialFunction=&MF_FX_EmissiveAdjustments}
  PixelNormalWS()
  Reroute(None=@n16.RGB)
  n22 = MaterialFunctionCall(
   Alpha=@n9
   Min=ScalarParameter() {defaultValue=0.3, parameterName=SmoothStepMin}
   Max=ScalarParameter() {defaultValue=0.7, parameterName=SmoothStepMax}
  ) {materialFunction=&SmoothStep}
  n23 = DynamicParameter() {defaultValue=<rgba 1 3 1 1>, paramNames [FadeIn,OpacityPower,Param3,Param4]}
  n24 = TextureCoordinate()
  n25 = TextureSampleParameter2D(UVs=LinearInterpolate(
    A=@n24
    B=MaterialFunctionCall(XY=@n24) {materialFunction=&Swizzle}.YX
    Alpha=ScalarParameter() {parameterName=Swizzle Mask UVs}
   )) {
   parameterName=OpacityMask
   samplerType=SAMPLERTYPE_Alpha
   texture=/Game/Gadgets/Assets/VinderTech_Umbrella/Umbrella_Storm/Textures/TopSwirlsOpacity
  }
  Reroute(None=Multiply(A=Min(A=@n22.Result, B=@n23.FadeIn), B=@n25.G))
  Reroute()
  n26 = ScalarParameter() {defaultValue=1000, parameterName=Freq}
  n27 = Add(
   A=Multiply(A=Time(), B=ScalarParameter() {defaultValue=1, parameterName=WPOSpeed})
   B=ComponentMask(None=Divide(A=@n1, B=@n26) {constB=1000}) {g=true}
  )
  Reroute(None=Multiply(A=VectorParameter() {parameterName=NoiseDirection, defaultValue=<rgba 0 0 1 0>}.RGB, B=Sine(None=@n27)))
  TextureCoordinate()
  VertexNormalWS()
  WorldPosition()
 ]
}
